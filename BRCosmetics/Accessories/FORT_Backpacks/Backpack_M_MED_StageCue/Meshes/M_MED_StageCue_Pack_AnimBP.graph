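# Animation Blueprint for the StageCue backpack that tilts the pack bone and toggles a material element for a front-end idle.
# The AnimGraph adds a fixed bone-space roll to the input pose.
# In the front end, each update sets FE Idle Active from the character's StageCue meta tag and the FrontEnd_Idle_Bypass curve.
# That flag drives the Hide Element 10 material parameter on update, customization load and weapon change.
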AnimBlueprint /BRCosmetics/Accessories/FORT_Backpacks/Backpack_M_MED_StageCue/Meshes/M_MED_StageCue_Pack_AnimBP
target_skeleton=&/M_MED_StageCue_Pack_Skeleton
graphs [
 AnimGraph {nodes [
   n1 = "Transform (Modify) Bone - Bone: pack"(ComponentPose=Local To Component(LocalPose=Input Pose()), Translation="0,0,0", Rotation="0,0,20") {
    rotationMode=BMM_Additive
    rotationSpace=BCS_BoneSpace
    alphaBoolBlend {blendOption=Linear}
    boneToModify {boneName=pack}
    rotation {roll=20}
   }
   Output Pose(Result=Component To Local(ComponentPose=@n1)) {layerGroup=None}
  ]}
 EventGraph {nodes [
   Event BlueprintUpdateAnimation()
   > n1 = Sequence()
   Branch(execute=[@n1.then_0, @n9.OutputPin], Condition=Get bIsInFrontEnd())
   n2 = Make Array("[0]"=(TagName="Cosmetics.MetaTags.StageCue"))
   n3 = GetCurveValue(self=GetOwnerFrontEndAnimInstance(), CurveName=FrontEnd_Idle_Bypass)
   > n4 = Branch(Condition=AND Boolean(A=CharacterItemDefinitionHasAnyMetaTag(Tags=@n2), B="float >= float"(A=@n3, B=1)))
   > Set FE Idle Active(FE Idle Active:bool=true) -> (Output_Get:bool)
   Set FE Idle Active(execute=@n4.else, FE Idle Active=false)
   n5 = Get Dynamic Material Instance(execute=[@n7.then, @n8.then, @n1.then_1], SkeletalMeshComponent=GetPartSkeletalMeshComponent())
   > n6 = Branch(Condition=Get FE Idle Active() -> (FE Idle Active:bool))
   > SetScalarParameterValue(self=@n5, ParameterName=Hide Element 10, Value=1)
   SetScalarParameterValue(execute=@n6.else, self=@n5, ParameterName=Hide Element 10, Value=0)
   n7 = Event OnCharacterCustomizationLoaded()
   n8 = Event OnWeaponChanged()
   n9 = Reroute Node()
   TryGetPawnOwner()
   PickaxeItemDefinitionHasAnyMetaTag(Tags=Make Array("[0]"=(TagName="Cosmetics.MetaTags.StageCue")), bRequiresEquipped=true)
  ]}
]
variables [FE Idle Active {type=boolean}]
